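# Sound cue named for a TV static warning layer in Athena, filed with interior prop sounds, but it has no root or sound nodes, so it has no audio to play.
# It keeps only playback settings: reduced volume, the Athena TV prop attenuation and the ambient TileMix sound class.
# A concurrency override refuses new instances at the limit instead of stopping playing ones.
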SoundCue /Game/Sounds/Fort_Prop_Sounds/Interior/TV_StaticLayerWarning_Athena_Cue
root=null
nodes []
properties {
 attenuationSettings=/Game/Sounds/Attenuation/Fort_Prop_TV_Athena_Att
 bOverrideConcurrency=true
 maxDistance=500
 soundClassObject=/Game/Sounds/FortSoundClassesAndModes/Fort_Ambient_TileMix
 volumeMultiplier=0.55
 attenuationOverrides {lPFRadiusMin=0}
 concurrencyOverrides {resolutionRule=PreventNew}
}
