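# Sound cue for a looping TV prop ambience in Athena, named as static and playing TV day-loop waves.
# A quality-level node picks between two looping wave players, sending the first two levels to the standard wave and the last two to the "opt" variant.
# It uses TV-specific attenuation and the ambient tile-mix sound class at reduced volume, with a concurrency override that prevents new instances at the limit.
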SoundCue /Game/Sounds/Fort_Prop_Sounds/Interior/TV_StaticLoop_Athena_Cue
root=QualityLevel_0
nodes [
 {id=QualityLevel_0, children [WavePlayer_0,WavePlayer_0,WavePlayer_1,WavePlayer_1]}
 {id=WavePlayer_0, properties {bLooping=true, soundWaveAssetPtr=&/TV_Day_Loop_01}}
 {id=WavePlayer_1, properties {bLooping=true, soundWaveAssetPtr=&/TV_Day_Loop_01_opt}}
]
properties {
 attenuationSettings=/Game/Sounds/Attenuation/Fort_Prop_TV_Athena_Att
 bOverrideConcurrency=true
 maxDistance=500
 soundClassObject=/Game/Sounds/FortSoundClassesAndModes/Fort_Ambient_TileMix
 volumeMultiplier=0.375
 attenuationOverrides {lPFRadiusMin=0}
 concurrencyOverrides {resolutionRule=PreventNew}
}
